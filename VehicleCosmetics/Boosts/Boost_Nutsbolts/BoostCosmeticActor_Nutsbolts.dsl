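; Boost cosmetic actor Blueprint for the Nutsbolts vehicle boost cosmetic.
; Its cosmetic-applied event and construction script only call their parent versions, so neither adds logic of its own.
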(event Vehicle|Cosmetics|EventOnCosmeticApplied
  (|CallParentFunction))

(fn ConstructionScript
  (|CallParentFunction))
