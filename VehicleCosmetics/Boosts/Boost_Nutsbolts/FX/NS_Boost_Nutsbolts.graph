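# Niagara system for the Nutsbolts vehicle boost cosmetic, spraying animated debris sprites and short-lived spark streaks from skeletal mesh sockets.
# The debris emitter merges two spawn groups, cooling from glowing, per-paint-variant colors to grey while falling and bouncing.
# User.bIsFrontendPreview switches debris between two collision modules, the preview one using analytical plane settings.
# The ember emitter adds bright, heavy-gravity sparks tinted by the painted color.
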NiagaraSystem /VehicleCosmetics/Boosts/Boost_Nutsbolts/FX/NS_Boost_Nutsbolts
dependencies {
 usedDataInterfaces [
  &NiagaraDataInterfaceCurve
  &NiagaraDataInterfaceVectorCurve
  &NiagaraDataInterfaceSpriteRendererInfo
  &NiagaraDataInterfaceColorCurve
  &NiagaraDataInterfaceArrayFloat2
  &NiagaraDataInterfaceSkeletalMesh
 ]
 usedDynamicInputs []
 usedModules [
  &self:SystemSpawnScript.SystemScriptSource.SystemScriptGraph.NiagaraNodeAssignment_0.SetVariables_#1
  &SystemState
 ]
}
emitters {
 NE_VC_Boost_Sprite_Parent001 {
  emitterUpdateScript [
   &EmitterState()
   &VC_FilteredSocketCount()
   &SpawnRate(SpawnRate=40, Use Spawn Probability=false)
   &SpawnRate(SpawnGroup=1, SpawnRate=40, Use Spawn Probability=false) {module=SpawnRate001}
  ]
  particleSpawnScript [
   &VC_PaintedVariantsOverride(DefaultColor=&ColorFromCurve(ColorCurve=DataInterface() {propertyValues {
       AlphaCurve {keys [<at 0 1 RCIM_Cubic>,<at 1 0 RCIM_Cubic>]}
       BlueCurve {keys [<at 0 0.3 RCIM_Cubic>, {arriveTangent=0.8115, interpMode=RCIM_Cubic, leaveTangent=0.8115, time=0.2255, value=0.1}, <at 0.713 0.8786>]}
       GreenCurve {keys [<at 0 0.9 RCIM_Cubic>, {arriveTangent=-1.2623, interpMode=RCIM_Cubic, leaveTangent=-1.2623, time=0.2255, value=0.3}, <at 0.713 0>]}
       RedCurve {keys [<at 0 3 RCIM_Cubic>, {arriveTangent=-2.8052, interpMode=RCIM_Cubic, leaveTangent=-2.8052, time=0.2255, value=1}, <at 0.713 1>]}
      }})) {enabled=false}
   /SetVariables(Particles.NewRenderTag:NiagaraInt32=&SpawnGroup()) {isSetParameters=true}
   &InitializeParticle(
    Color Mode=::NewEnumerator0
    Lifetime=/VehicleCosmetics/FX/Niagara/Modules/MergeEmitters/VC_FloatBySpawnGroup(
     SpawnGroup or ExecIndex:NiagaraInt32=0
     CustomInt:NiagaraInt32=0
     MaximumSpawnGroups:NiagaraInt32=3
     Float SG 0:NiagaraFloat=1.5
     Float SG 1:NiagaraFloat=1.5
     Float SG 2:NiagaraFloat=&RandomRangeFloat(Minimum=0.15, Maximum=0.2)
     Float SG 3:NiagaraFloat=0
     Float SG 4:NiagaraFloat=0
     Float SG 5:NiagaraFloat=0
    )
    Sprite Rotation Mode=::NewEnumerator2
    Sprite Size Max=&SelectVector2DFromArray(
     Direct Array Index=&SpawnGroup()
     Vector 2D Selection Array=DataInterface() {propertyValues {FloatData [<21.6 21.6>, <27 27>, <32 96>]}}
    )
    Sprite Size Min=&SelectVector2DFromArray(
     Direct Array Index=&SpawnGroup()
     Vector 2D Selection Array=DataInterface() {propertyValues {FloatData [<13.5 13.5>, <21.6 21.6>, <12 72>]}}
    )
    Sprite Size Mode=::NewEnumerator2
    UsePositionOffset=false
    Write Material Random=true
   )
   &PartitionParticles(Number Of Partitions=$Emitter.PortCount)
   &SkeletalMeshLocation(
    Bone / Socket Index=$Particles.PartitionParticles.Partition
    Bone / Socket Sampling Mode=::NewEnumerator5
    MaskSpawnGroup=false
    Mesh Sampling Type=::NewEnumerator6
    Sampled Position Offset=&AddVectorToPosition(Position=<-20 0 0>, Vector=$User.DTCarOffset)
    Skeletal Mesh=$System.SkeletalMesh
   )
   &AddVelocity(Velocity=&RandomRangeVector(Minimum=<-500 -120 0>, Maximum=<-500 120 200>))
   &Color(Color=&MakeCustomLinearColorFromBool(Bool=$User.bIsPainted, True Color=$User.PaintedColor, False Color=<rgba 1 0.3 0.1 0.75>)) {enabled=false}
  ]
  particleUpdateScript [
   &ParticleState()
   &VC_PaintedVariantsOverride(
    Black=&ColorFromCurve(ColorCurve=DataInterface() {propertyValues {
       AlphaCurve {keys [<at 0 1 RCIM_Cubic>,<at 1 1 RCIM_Cubic>]}
       BlueCurve {keys [<at 0 0 RCIM_Cubic>,<at 0.2 0.019>,<at 0.35 0.5 RCIM_Cubic>]}
       GreenCurve {keys [<at 0 0 RCIM_Cubic>,<at 0.2 0.019>,<at 0.35 0.5 RCIM_Cubic>]}
       RedCurve {keys [<at 0 0 RCIM_Cubic>,<at 0.2 0.019>,<at 0.35 0.5 RCIM_Cubic>]}
      }})
    BurntSienna=&ColorFromCurve(ColorCurve=DataInterface() {propertyValues {
       AlphaCurve {keys [<at 0 1 RCIM_Cubic>,<at 1 1 RCIM_Cubic>]}
       BlueCurve {keys [<at 0 0.048 RCIM_Cubic>,<at 0.17 0.028>,<at 0.32 0.5 RCIM_Cubic>]}
       GreenCurve {keys [<at 0 0.296 RCIM_Cubic>,<at 0.17 0.1727>,<at 0.32 0.5 RCIM_Cubic>]}
       RedCurve {keys [<at 0 1.2 RCIM_Cubic>,<at 0.17 0.7>,<at 0.32 0.5 RCIM_Cubic>]}
      }})
    Cobalt=&ColorFromCurve(ColorCurve=DataInterface() {propertyValues {
       AlphaCurve {keys [<at 0 1 RCIM_Cubic>,<at 1 1 RCIM_Cubic>]}
       BlueCurve {keys [<at 0 36 RCIM_Cubic>,<at 0.17 7>,<at 0.32 0.5 RCIM_Cubic>]}
       GreenCurve {keys [<at 0 0.684 RCIM_Cubic>,<at 0.17 0.133>,<at 0.32 0.5 RCIM_Cubic>]}
       RedCurve {keys [<at 0 0 RCIM_Cubic>,<at 0.17 0>,<at 0.32 0.5 RCIM_Cubic>]}
      }})
    Crimson=&ColorFromCurve(ColorCurve=DataInterface() {propertyValues {
       AlphaCurve {keys [<at 0 1 RCIM_Cubic>,<at 1 1 RCIM_Cubic>]}
       BlueCurve {keys [<at 0 0.1688 RCIM_Cubic>,<at 0.15 0.1037>,<at 0.3 0.5 RCIM_Cubic>]}
       GreenCurve {keys [<at 0 0.1688 RCIM_Cubic>,<at 0.15 0>,<at 0.3 0.5 RCIM_Cubic>]}
       RedCurve {keys [<at 0 5.4 RCIM_Cubic>,<at 0.15 4>,<at 0.3 0.5 RCIM_Cubic>]}
      }})
    DefaultColor=&ColorFromCurve(ColorCurve=DataInterface() {propertyValues {
       AlphaCurve {keys [<at 0 1 RCIM_Cubic>,<at 1 1 RCIM_Cubic>]}
       BlueCurve {keys [<at 0 1.2 RCIM_Cubic>, {arriveTangent=-2, interpMode=RCIM_Cubic, leaveTangent=-2, time=0.2, value=0.4}, <at 0.35 0.5>]}
       GreenCurve {keys [<at 0 3.6 RCIM_Cubic>, {arriveTangent=-8.8571, interpMode=RCIM_Cubic, leaveTangent=-8.8571, time=0.2, value=1.2}, <at 0.35 0.5>]}
       RedCurve {keys [<at 0 12 RCIM_Cubic>, {arriveTangent=-32.8571, interpMode=RCIM_Cubic, leaveTangent=-32.8571, time=0.2, value=6}, <at 0.35 0.5>]}
      }})
    ForestGreen=&ColorFromCurve(ColorCurve=DataInterface() {propertyValues {
       AlphaCurve {keys [<at 0 1 RCIM_Cubic>,<at 1 1 RCIM_Cubic>]}
       BlueCurve {keys [<at 0 0.14 RCIM_Cubic>,<at 0.16 0.075>,<at 0.31 0.5 RCIM_Cubic>]}
       GreenCurve {keys [<at 0 2.8 RCIM_Cubic>,<at 0.16 1.5>,<at 0.31 0.5 RCIM_Cubic>]}
       RedCurve {keys [<at 0 0 RCIM_Cubic>,<at 0.16 0>,<at 0.31 0.5 RCIM_Cubic>]}
      }})
    Grey=<rgba 0.5 0.5 0.5 1>
    Lime=&ColorFromCurve(ColorCurve=DataInterface() {propertyValues {
       AlphaCurve {keys [<at 0 1 RCIM_Cubic>,<at 1 1 RCIM_Cubic>]}
       BlueCurve {keys [<at 0 0 RCIM_Cubic>,<at 0.2 0>,<at 0.33 0.5 RCIM_Cubic>]}
       GreenCurve {keys [<at 0 9 RCIM_Cubic>,<at 0.2 2.4>,<at 0.33 0.5 RCIM_Cubic>]}
       RedCurve {keys [<at 0 4.5 RCIM_Cubic>,<at 0.2 1.2>,<at 0.33 0.5 RCIM_Cubic>]}
      }})
    Orange=&ColorFromCurve(ColorCurve=DataInterface() {propertyValues {
       AlphaCurve {keys [<at 0 1 RCIM_Cubic>,<at 1 1 RCIM_Cubic>]}
       BlueCurve {keys [<at 0 0 RCIM_Cubic>,<at 0.2 0>,<at 0.35 0.5 RCIM_Cubic>]}
       GreenCurve {keys [<at 0 2.7 RCIM_Cubic>,<at 0.2 0.9>,<at 0.35 0.5 RCIM_Cubic>]}
       RedCurve {keys [<at 0 9 RCIM_Cubic>,<at 0.2 3>,<at 0.35 0.5 RCIM_Cubic>]}
      }})
    Pink=&ColorFromCurve(ColorCurve=DataInterface() {propertyValues {
       AlphaCurve {keys [<at 0 1 RCIM_Cubic>,<at 1 1 RCIM_Cubic>]}
       BlueCurve {keys [<at 0 5.1287 RCIM_Cubic>,<at 0.2 1.9782>,<at 0.34 0.5 RCIM_Cubic>]}
       GreenCurve {keys [<at 0 1.8574 RCIM_Cubic>,<at 0.2 0.7164>,<at 0.34 0.5 RCIM_Cubic>]}
       RedCurve {keys [<at 0 7 RCIM_Cubic>,<at 0.2 2.7>,<at 0.34 0.5 RCIM_Cubic>]}
      }})
    Purple=&ColorFromCurve(ColorCurve=DataInterface() {propertyValues {
       AlphaCurve {keys [<at 0 1 RCIM_Cubic>,<at 1 1 RCIM_Cubic>]}
       BlueCurve {keys [<at 0 9 RCIM_Cubic>,<at 0.18 7>,<at 0.33 0.5 RCIM_Cubic>]}
       GreenCurve {keys [<at 0 0 RCIM_Cubic>,<at 0.18 0>,<at 0.33 0.5 RCIM_Cubic>]}
       RedCurve {keys [<at 0 1.5 RCIM_Cubic>,<at 0.18 1.1667>,<at 0.33 0.5 RCIM_Cubic>]}
      }})
    Saffron=&ColorFromCurve(ColorCurve=DataInterface() {propertyValues {
       AlphaCurve {keys [<at 0 1 RCIM_Cubic>,<at 1 1 RCIM_Cubic>]}
       BlueCurve {keys [<at 0 0 RCIM_Cubic>,<at 0.18 0>,<at 0.33 0.5 RCIM_Cubic>]}
       GreenCurve {keys [<at 0 7.6667 RCIM_Cubic>,<at 0.18 5.3667>,<at 0.33 0.5 RCIM_Cubic>]}
       RedCurve {keys [<at 0 10 RCIM_Cubic>,<at 0.18 7>,<at 0.33 0.5 RCIM_Cubic>]}
      }})
    SkyBlue=&ColorFromCurve(ColorCurve=DataInterface() {propertyValues {
       AlphaCurve {keys [<at 0 1 RCIM_Cubic>,<at 1 1 RCIM_Cubic>]}
       BlueCurve {keys [<at 0 7 RCIM_Cubic>,<at 0.17 2.7>,<at 0.31 0.5 RCIM_Cubic>]}
       GreenCurve {keys [<at 0 4.6447 RCIM_Cubic>,<at 0.17 1.7915>,<at 0.31 0.5 RCIM_Cubic>]}
       RedCurve {keys [<at 0 0 RCIM_Cubic>,<at 0.17 0>,<at 0.31 0.5 RCIM_Cubic>]}
      }})
    TitaniumWhite=&ColorFromCurve(ColorCurve=DataInterface() {propertyValues {
       AlphaCurve {keys [<at 0 1 RCIM_Cubic>,<at 1 1 RCIM_Cubic>]}
       BlueCurve {keys [<at 0 9 RCIM_Cubic>,<at 0.15 7>,<at 0.3 0.5 RCIM_Cubic>]}
       GreenCurve {keys [<at 0 9 RCIM_Cubic>,<at 0.15 7>,<at 0.3 0.5 RCIM_Cubic>]}
       RedCurve {keys [<at 0 9 RCIM_Cubic>,<at 0.15 7>,<at 0.3 0.5 RCIM_Cubic>]}
      }})
   ) {module=VC_PaintedVariantsOverride001}
   &ScaleSpriteSize(
    Scale Sprite Size Mode=::NewEnumerator0
    Uniform Scale Factor=&VC_FloatBySpawnGroup(
     SpawnGroup or ExecIndex:NiagaraInt32=0
     CustomInt:NiagaraInt32=0
     MaximumSpawnGroups:NiagaraInt32=3
     Float SG 0:NiagaraFloat=2
     Float SG 1:NiagaraFloat=2
     Float SG 2:NiagaraFloat=&FloatFromCurve(FloatCurve=DataInterface() {propertyValues {Curve {keys [{tangentMode=RCTM_User, time=0, value=0}, <at 0.1 1>, {tangentMode=RCTM_User, time=1, value=0}]}}})
     Float SG 3:NiagaraFloat=0
     Float SG 4:NiagaraFloat=0
     Float SG 5:NiagaraFloat=0
    )
   )
   &GravityForce(Gravity=&RandomRangeVector(Minimum=<0 0 -256>, Maximum=<0 0 -355>))
   &InheritSourceMovement(
    Applied Position Scale=&VectorFromCurve(VectorCurve=DataInterface() {propertyValues {XCurve {keys [<at 0.1 0.7>,<at 0.5 0>]}, YCurve {keys [<at 0.1 0.7>,<at 0.5 0>]}, ZCurve {keys [<at 0.1 0.7>,<at 0.5 0>]}}})
    ApplyForce=false
    ApplyPosition=true
    ApplyPositionFalloff=false
   )
   &Collision(
    Advanced Aging Rate=5
    Age Colliding Particles=true
    Collision Enabled=HlslExpression() {hlslExpression="User.bIsFrontendPreview == true ? false : true"}
    Control Roll On Collision=true
    Friction=1.5
    Friction During a Bounce:NiagaraFloat=&RandomRangeFloat(Minimum=0.9, Maximum=1.2)
    KillOnCollision=false
    Module.Control Rest State Via Surface Normal=false
    Module.EnableMaxCollisionCount=false
    Module.Randomize Collision Normal Vector=false
    Restitution=&RandomRangeFloat(Minimum=0, Maximum=0.3)
    ScaleRateOfRotation=false
   )
   &Collision(
    Advanced Aging Rate=5
    Age Colliding Particles=true
    Analytical Collision Plane Position 1:Vector3f=<0 0 20>
    Analytical Collision Plane Position 2:Vector3f=<0 0 20>
    CPU Collision Type=::NewEnumerator3
    Collision Enabled=HlslExpression() {hlslExpression="User.bIsFrontendPreview == true ? true : false"}
    Control Roll On Collision=true
    Friction=1.5
    Friction During a Bounce:NiagaraFloat=&RandomRangeFloat(Minimum=0.9, Maximum=1.2)
    KillOnCollision=false
    Module.Control Rest State Via Surface Normal=false
    Module.EnableMaxCollisionCount=false
    Module.Randomize Collision Normal Vector=false
    Restitution=&RandomRangeFloat(Minimum=0, Maximum=0.3)
    ScaleRateOfRotation=false
   ) {module=Collision001}
   &SolveForcesAndVelocity(Clamp Velocity=false, Limit Acceleration=false)
   &SubUVAnimation(
    Frames Per Second=60
    Output Alternate Renderer Binding=true
    Playback Mode:ENiagara_FPSPlayrate=::NewEnumerator1
    Playback Scale Curve=DataInterface() {propertyValues {Curve {keys [
        {interpMode=RCIM_Cubic, tangentMode=RCTM_User, time=0.25, value=1}
        {arriveTangent=-1.2656, interpMode=RCIM_Cubic, leaveTangent=-1.2656, tangentMode=RCTM_User, time=1, value=0}
       ]}}}
    "Sprite Renderer "=DataInterface() {propertyValues {SpriteRenderer=&self:NE_VC_Boost_Sprite_Parent001_0.NiagaraSpriteRendererProperties_0}}
    SubUV Animation Mode:ENiagara_SubUVLookupModeV2=::NewEnumerator3
    UseEndFrame=false
    UsePlayRateScale=false
    UseStartFrame=false
   )
   &SubUVAnimation(
    Frames Per Second=30
    Output Alternate Renderer Binding=true
    Playback Mode:ENiagara_FPSPlayrate=::NewEnumerator1
    Playback Scale Curve=DataInterface() {propertyValues {Curve {keys [
        {interpMode=RCIM_Cubic, tangentMode=RCTM_User, time=0.25, value=1}
        {arriveTangent=-1.2656, interpMode=RCIM_Cubic, leaveTangent=-1.2656, tangentMode=RCTM_User, time=1, value=0}
       ]}}}
    "Sprite Renderer "=DataInterface() {propertyValues {SpriteRenderer=&self:NE_VC_Boost_Sprite_Parent001_0.NiagaraSpriteRendererProperties_2}}
    SubUV Animation Mode:ENiagara_SubUVLookupModeV2=::NewEnumerator3
    UseEndFrame=false
    UsePlayRateScale=false
    UseStartFrame=false
   ) {module=SubUVAnimation001}
   &ScaleColor(
    Scale Alpha=&MakeCustomFloatFromBool(
     Bool=$User.bIsPainted
     True Float=&MakeCustomFloatFromBool(
      Bool=HlslExpression() {hlslExpression="User.ColorVariantID == 2 ? true : false"}
      True Float=&FloatFromCurve(FloatCurve=DataInterface() {propertyValues {Curve {keys [<at 0 5>,<at 0.3 0.75>]}}})
      False Float=&FloatFromCurve(FloatCurve=DataInterface() {propertyValues {Curve {keys [<at 0 1>,<at 0.3 0.75>]}}})
     )
     False Float=&FloatFromCurve(FloatCurve=DataInterface() {propertyValues {Curve {keys [<at 0 1>,<at 0.3 0.75>]}}})
    )
    Scale RGB=&VectorFromCurve(
     VectorCurve=DataInterface() {propertyValues {
       XCurve {keys [<at 0 3>, {arriveTangent=-13.3333, time=0.2, value=1}, <at 0.35 0 RCIM_Cubic>]}
       YCurve {keys [<at 0 3>, {arriveTangent=-13.3333, time=0.2, value=1}, <at 0.35 0 RCIM_Cubic>]}
       ZCurve {keys [<at 0 3>, {arriveTangent=-13.3333, time=0.2, value=1}, <at 0.35 0 RCIM_Cubic>]}
      }}
     Scale Curve=&VectorFromFloat(Value=$User.ScaleColorFactor)
    )
    ScaleA=false
    ScaleRGB=true
   ) {enabled=false}
   &DynamicMaterialParameters(
    Index 0 Param 1=1
    Index 0 Param 2=1
    Module.Index0Float0Write=true
    Module.Index0Float1Write=true
    Module.Index0Float2Write=true
    Module.Index0Float3Write=true
   )
  ]
  rendererClasses [&NiagaraSpriteRendererProperties]
  renderers [
   {rendererClass=&NiagaraSpriteRendererProperties}
   {rendererClass=&NiagaraSpriteRendererProperties, rendererIndex=1}
  ]
 }
 NE_VC_Embers001 {
  emitterUpdateScript [
   &EmitterState()
   &VC_FilteredSocketCount()
   &SpawnRate(Spawn Probability=1.5, SpawnRate=512, Use Spawn Probability=false)
  ]
  particleSpawnScript [
   &VC_PaintedVariantsOverride() {enabled=false}
   &InitializeParticle(
    Color=&MakeCustomLinearColorFromBool(
     Bool=$User.bIsPainted
     True Color=$User.PaintedColor
     False Color=&RandomRangeLinearColor(Minimum=<rgba 2 0.35 0.1 0.01>, Maximum=<rgba 2 0.35 0.1 0.03>)
    )
    Lifetime Max=0.2
    Lifetime Min=0.15
    Lifetime Mode=::NewEnumerator1
    Sprite Rotation Mode=::NewEnumerator2
    Sprite Size=&RandomRangeVector2D(Minimum=<1.8 72>, Maximum=<5 96>)
    Sprite Size Mode=::NewEnumerator4
    UsePositionOffset=false
    Write Material Random=true
   )
   &PartitionParticles(Number Of Partitions=$Emitter.PortCount)
   &SkeletalMeshLocation(
    Bone / Socket Index=$Particles.PartitionParticles.Partition
    Bone / Socket Sampling Mode=::NewEnumerator5
    MaskSpawnGroup=false
    Mesh Sampling Type=::NewEnumerator6
    Sampled Position Offset=&AddVectorToPosition(Position=<-6 0 0>, Vector=$User.DTCarOffset)
    Skeletal Mesh=$System.SkeletalMesh
   )
   &AddVelocity(Velocity=&RandomRangeVector(Minimum=<-288 -550 360>, Maximum=<-1776 550 720>))
   &ShapeLocation(MaskSpawnGroup=false, Override Seed=false, Sphere Radius=24, UseOwnerScale=true)
  ]
  particleUpdateScript [
   &ParticleState()
   &ScaleSpriteSize(Uniform Curve Sprite Scale=DataInterface() {propertyValues {Curve {keys [{tangentMode=RCTM_User, time=0, value=0}, <at 0.1 1>, {tangentMode=RCTM_User, time=1, value=0}]}}})
   &ScaleVelocity(Velocity Scale=&VectorFromCurve(VectorCurve=DataInterface() {propertyValues {XCurve {keys [<at 0 3>,<at 1 0.5>]}, YCurve {keys [<at 0 3>,<at 1 0.5>]}, ZCurve {keys [<at 0 3>,<at 1 0.5>]}}}))
   &AccelerationForce(Acceleration=<0 0 -2560>) {enabled=false}
   &GravityForce(Gravity=&VectorFromCurve(VectorCurve=DataInterface() {propertyValues {XCurve {keys [<at 0 0>,<at 1 0>]}, YCurve {keys [<at 0 0>,<at 1 0>]}, ZCurve {keys [<at 0.5 -29400 RCIM_Cubic>,<at 1 -44100 RCIM_Cubic>]}}}))
   &SolveForcesAndVelocity(Clamp Velocity=false, Limit Acceleration=false)
   &ScaleColor(
    Scale Alpha=&MakeCustomFloatFromBool(Bool=$User.bIsPainted, True Float=&RandomRangeFloat(Minimum=0.01, Maximum=0.03), False Float=1)
    Scale RGB=&VectorFromCurve(
     VectorCurve=DataInterface() {propertyValues {XCurve {keys [<at 0 512>,<at 0.5 32>]}, YCurve {keys [<at 0 512>,<at 0.5 32>]}, ZCurve {keys [<at 0 512>,<at 0.5 32>]}}}
     Scale Curve=&VectorFromFloat(Value=$User.SparksColorScale)
    )
    ScaleA=true
    ScaleRGB=true
   )
   &SubUVAnimation(
    "Sprite Renderer "=DataInterface() {propertyValues {SpriteRenderer=&self:NE_VC_Embers001_0.NiagaraSpriteRendererProperties_0}}
    SubUV Animation Mode:ENiagara_SubUVLookupModeV2=::NewEnumerator2
    UseEndFrame=false
    UseStartFrame=false
   )
  ]
  rendererClasses [&NiagaraSpriteRendererProperties]
  renderers [{rendererClass=&NiagaraSpriteRendererProperties}]
 }
}
summary {
 systemName=NS_Boost_Nutsbolts
 emitters [
  {emitterName=NE_VC_Embers001, rendererClasses [&NiagaraSpriteRendererProperties]}
  {emitterName=NE_VC_Boost_Sprite_Parent001, rendererClasses [&NiagaraSpriteRendererProperties]}
 ]
 userVariables [
  User.AlphaScale {type=NiagaraFloat, defaultValue=1, description=""}
  User.bIsFrontendPreview {type=NiagaraBool}
  User.bIsPainted {type=NiagaraBool}
  User.ColorVariantID {type=NiagaraInt32}
  User.DTCarOffset {type=Vector3f, description="", defaultValue {struct=Vector3f, value=<0 0 0>}}
  User.PaintedColor {type=LinearColor, description="", defaultValue {struct=LinearColor, value=<rgba 1 1 1 1>}}
  User.PaintedScalar {type=NiagaraFloat, defaultValue=1, description=""}
  User.PortCount {type=NiagaraInt32, defaultValue=1}
  User.ScaleColorFactor {type=NiagaraFloat, defaultValue=1, description=""}
  User.SparksColorScale {type=NiagaraFloat, defaultValue=2, description=""}
 ]
}
